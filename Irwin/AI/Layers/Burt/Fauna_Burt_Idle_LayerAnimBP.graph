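# Animation layer Blueprint that supplies the idle override for Burt, a prey fauna AI.
# On initialization it caches the main Burt_AnimBP to read its aim-target state.
# Fauna Idle holds the plain idle while an aim target is valid, otherwise randomly alternates idle and idle-break at equal odds, switching with inertialized blending.
# Fauna IdleTurn passes its input pose through unchanged, and the main AnimGraph is empty.
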AnimBlueprint /Irwin/AI/Layers/Burt/Fauna_Burt_Idle_LayerAnimBP
target_skeleton=/Irwin/AI/Base/Burt_Mammal_Invis_Skeleton
graphs [
 AnimGraph {nodes [Output Pose() {layerGroup=None}]}
 EventGraph {nodes [
   Event BlueprintInitializeAnimation()
   > n1 = Cast To Burt_AnimBP!(execute, Object=Get MainAnimBP() -> (MainAnimBP:Anim Instance Object Reference)) -> (then, CastFailed, AsBurt Anim BP:Burt Anim BP Object Reference) {is=title K2Node_DynamicCast}
   > Set BurtAnimBP(BurtAnimBP:Burt Anim BP Object Reference=@n1) -> (Output_Get:Burt Anim BP Object Reference)
  ]}
 Fauna Idle {nodes [
   Input Pose() {name=InPose Idle}
   n1 = Random Sequence Player!() -> (Pose:Pose Link Structure) {
    is=title AnimGraphNode_RandomPlayer
    entries [
     {
      chanceToPlay=0.5
      maxLoopCount=0
      maxPlayRate=1
      minLoopCount=0
      minPlayRate=1
      sequence=/Irwin/AI/Prey/Burt/Animations/Burt_Idle
      blendIn {blendOption=Linear, blendTime=0.35, customCurve=null}
     }
     {
      chanceToPlay=0.5
      maxLoopCount=0
      maxPlayRate=1
      minLoopCount=0
      minPlayRate=1
      sequence=/Irwin/AI/Prey/Burt/Animations/Burt_IdleBreak
      blendIn {blendOption=Linear, blendTime=0.35, customCurve=null}
     }
    ]
   }
   n2 = Get BurtAnimBP() -> (BurtAnimBP:Burt Anim BP Object Reference)
   Output Pose(Result=Blend Poses by bool(
     BlendPose_0=Sequence Player 'Burt_Idle'() {sequence=&Burt_Idle}
     BlendPose_1=@n1
     BlendTime_0=0.5
     BlendTime_1=0.5
     bActiveValue=Get bHasValidAimTarget!(self:Fort AIFauna Anim Instance Object Reference=@n2) -> (bHasValidAimTarget:bool)
    ) {blendType=Linear, transitionType=Inertialization}) {layerGroup=FaunaIdleOverride}
  ]}
 Fauna IdleTurn {nodes [Output Pose(Result=Input Pose() {name=InPose IdleTurn}) {layerGroup=FaunaIdleOverride}]}
 GetMainAnimBP {nodes [
   GetMainAnimBP() {is=title K2Node_FunctionEntry}
   > n1 = Sequence()
   > n2 = Cast To FortPawn!(execute, Object=TryGetPawnOwner()) -> (then, CastFailed, AsFort Pawn:Fort Pawn Object Reference) {is=title K2Node_DynamicCast}
   > n3 = Cast To Burt_AnimBP(Object=GetAnimInstance(self=Get Mesh(self=@n2)))
   > Set AnimBP(AnimBP:Burt Anim BP Object Reference=@n3) -> (Output_Get:Burt Anim BP Object Reference)
   Return Node!(execute=@n1.then_1, ReturnValue:Burt Anim BP Object Reference=Get AnimBP() -> (AnimBP:Burt Anim BP Object Reference))
  ]}
]
variables [BurtAnimBP {type=Burt_AnimBP_C, default=null}]
