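# Animation layer Blueprint for Burt fauna locomotion, targeting the Burt_Mammal_Invis skeleton.
# On initialization it casts the main AnimBP to Burt_AnimBP and caches it in BurtAnimBP, though no graph shown reads it.
# The Fauna Locomotion and Fauna LocomotionTurn layers in the FaunaLocomotionOverride group pass their input poses straight to output, adding no animation logic of their own.
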AnimBlueprint /Irwin/AI/Layers/Burt/Fauna_Burt_Locomotion_LayerAnimBP
target_skeleton=/Irwin/AI/Base/Burt_Mammal_Invis_Skeleton
graphs [
 AnimGraph {nodes [Output Pose() {layerGroup=None}]}
 EventGraph {nodes [
   Event BlueprintInitializeAnimation()
   > n1 = Cast To Burt_AnimBP!(execute, Object=Get MainAnimBP() -> (MainAnimBP:Anim Instance Object Reference)) -> (then, CastFailed, AsBurt Anim BP:Burt Anim BP Object Reference) {is=title K2Node_DynamicCast}
   > Set BurtAnimBP(BurtAnimBP:Burt Anim BP Object Reference=@n1) -> (Output_Get:Burt Anim BP Object Reference)
  ]}
 Fauna Locomotion {nodes [Output Pose(Result=Input Pose() {name=InPose Locomotion}) {layerGroup=FaunaLocomotionOverride}]}
 Fauna LocomotionTurn {nodes [Output Pose(Result=Input Pose() {name=InPose LocomotionTurn}) {layerGroup=FaunaLocomotionOverride}]}
]
variables [BurtAnimBP {type=Burt_AnimBP_C, default=null}]
